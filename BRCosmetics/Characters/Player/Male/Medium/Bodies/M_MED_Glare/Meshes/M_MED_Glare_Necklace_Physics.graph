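# Physics asset for the necklace on the medium male Glare character body, setting up collision and joint limits for its dangling chains and charms.
# A constraint chain runs from the necklace base to a parent link, then forks into durrr and heart chains, each ending in a charm.
# The durrr branch swings wider than the heart branch, and all joints have tight twist limits.
# Two large unconstrained spine_05 capsules provide upper-torso collision.
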PhysicsAsset /BRCosmetics/Characters/Player/Male/Medium/Bodies/M_MED_Glare/Meshes/M_MED_Glare_Necklace_Physics
bodies {
 dyn_durrr_chain [{shapeType=Capsule, length=1.3438, radius=2.25, shapeName=dyn_durrr_chain_capsule, center=<1.7 0 -1>, rotation=<rot 90 0 0>}]
 dyn_durrr_charm [{shapeType=Capsule, length=6.4062, radius=2.7812, shapeName=dyn_durrr_charm_capsule, center=<4 0 -1>, rotation=<rot 90 0 -80>}]
 dyn_heart_chain [{shapeType=Capsule, length=0.2812, radius=1.7238, shapeName=dyn_heart_chain_capsule, center=<1.4108 0 0>, rotation=<rot 90 0 0>}]
 dyn_heart_charm [{shapeType=Capsule, length=1.5938, radius=1.5167, shapeName=dyn_heart_charm_capsule, center=<2.2 0 -0.3>, rotation=<rot 90 0 0>}]
 dyn_necklace_base [{shapeType=Sphere, radius=0.5, shapeName=dyn_necklace_base_sphere}]
 dyn_necklace_parent [{
   shapeType=Capsule
   length=2.8438
   radius=3.1875
   shapeName=dyn_necklace_parent_capsule
   center=<4.15 0 -0.45>
   rotation=<rot 77.8879 180 180>
  }]
 spine_05 [
  {shapeType=Capsule, length=7.25, radius=16.25, shapeName=spine_05_capsule, center=<-5.5326 -4.7512 0>}
  {
   shapeType=Capsule
   length=10.9062
   radius=19.0938
   shapeName=spine_05_capsule
   center=<-19.8679 -6.348 0.2021>
   rotation=<rot -89.4173 179.9998 -180>
  }
 ]
}
constraints [
 {bone1Name=dyn_necklace_parent, bone2Name=dyn_necklace_base, swing1LimitDegrees=10, swing2LimitDegrees=20, twistLimitDegrees=3}
 {bone1Name=dyn_durrr_chain, bone2Name=dyn_necklace_parent, swing1LimitDegrees=35, swing2LimitDegrees=20, twistLimitDegrees=3}
 {bone1Name=dyn_durrr_charm, bone2Name=dyn_durrr_chain, swing1LimitDegrees=40, swing2LimitDegrees=20, twistLimitDegrees=3}
 {bone1Name=dyn_heart_chain, bone2Name=dyn_necklace_parent, swing1LimitDegrees=10, swing2LimitDegrees=10, twistLimitDegrees=3}
 {bone1Name=dyn_heart_charm, bone2Name=dyn_heart_chain, swing1LimitDegrees=10, swing2LimitDegrees=10, twistLimitDegrees=3}
]
