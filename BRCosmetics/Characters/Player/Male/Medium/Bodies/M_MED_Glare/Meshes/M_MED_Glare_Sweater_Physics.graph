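# Physics asset for the sweater on the M_MED_Glare male character body, defining collision and joints for its left and right dyn_sweater bones.
# Each side forms a three-body chain from a small base sphere through parent and 01 capsules, linked by narrow swing limits with looser twist at the outer joint.
# Pelvis and thigh capsules have no constraints, providing body volumes alongside the sweater chains.
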PhysicsAsset /BRCosmetics/Characters/Player/Male/Medium/Bodies/M_MED_Glare/Meshes/M_MED_Glare_Sweater_Physics
bodies {
 dyn_sweater_01_l [{shapeType=Capsule, length=1.6216, radius=1.6998, shapeName=dyn_sweater_01_l_capsule, center=<2.3 0 -0.63>, rotation=<rot 90 0 0>}]
 dyn_sweater_01_r [{shapeType=Capsule, length=1.6216, radius=1.6998, shapeName=dyn_sweater_01_l_capsule, center=<2.3 0 -0.68>, rotation=<rot 90 0 0>}]
 dyn_sweater_base_l [{shapeType=Sphere, radius=0.5, shapeName=dyn_sweater_base_l_sphere}]
 dyn_sweater_base_r [{shapeType=Sphere, radius=0.5, shapeName=dyn_sweater_base_r_sphere}]
 dyn_sweater_parent_l [{shapeType=Capsule, length=3, radius=1.375, shapeName=dyn_sweater_base_r_capsule, center=<2.0754 0 0>, rotation=<rot 90 0 0>}]
 dyn_sweater_parent_r [{shapeType=Capsule, length=3, radius=1.375, shapeName=dyn_sweater_base_r_capsule, center=<2.0754 0 0>, rotation=<rot 90 0 0>}]
 pelvis [{shapeType=Capsule, length=11.2812, radius=13.0312, shapeName=pelvis_capsule, center=<-11.4252 3.3347 0>}]
 thigh_l [{shapeType=Capsule, length=30.0938, radius=8.2812, shapeName=thigh_l_capsule, center=<-20.3358 0 0.519>, rotation=<rot 87.9084 0.0002 0.0004>}]
 thigh_r [{shapeType=Capsule, length=29.2188, radius=7.8125, shapeName=thigh_r_capsule, center=<22.0971 0 -0.5539>, rotation=<rot 90 0 0>}]
}
constraints [
 {bone1Name=dyn_sweater_parent_r, bone2Name=dyn_sweater_base_r, swing1LimitDegrees=10, swing2LimitDegrees=10, twistLimitDegrees=3}
 {bone1Name=dyn_sweater_parent_l, bone2Name=dyn_sweater_base_l, swing1LimitDegrees=10, swing2LimitDegrees=10, twistLimitDegrees=3}
 {bone1Name=dyn_sweater_01_r, bone2Name=dyn_sweater_parent_r, swing1LimitDegrees=10, swing2LimitDegrees=10, twistLimitDegrees=45}
 {bone1Name=dyn_sweater_01_l, bone2Name=dyn_sweater_parent_l, swing1LimitDegrees=10, swing2LimitDegrees=10, twistLimitDegrees=45}
]
