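# Physics asset for the Glare male medium character's face accessory mesh, defining bodies and joints for dynamic hair.
# A large head capsule sits among capsules on the hair and strand bones, with small spheres at each chain's base bone.
# Constraints link each front, side, back and loose-strand chain bone to bone from its base.
# Swing limits are tight at front and sides and widest at center back, while twist stays tight throughout.
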PhysicsAsset /BRCosmetics/Characters/Player/Male/Medium/Bodies/M_MED_Glare/Meshes/Parts/M_MED_Glare_FaceAcc_Physics
bodies {
 dyn_hair_01_bk_side_l [{
   shapeType=Capsule
   length=1.795
   radius=0.8719
   shapeName=dyn_hair_01_bk_side_l_capsule
   center=<1.162 0 0.4567>
   rotation=<rot 67.1598 180 180>
  }]
 dyn_hair_01_bk_side_r [{shapeType=Capsule, length=1.1562, radius=1.25, shapeName=dyn_hair_01_bk_side_r_capsule, center=<1.3822 0 0>, rotation=<rot 90 0 0>}]
 dyn_hair_01_front_l [{shapeType=Capsule, length=2.136, radius=1.1535, shapeName=dyn_hair_01_front_l_capsule, center=<1.6202 0 0>, rotation=<rot 90 0 0>}]
 dyn_hair_01_front_r [{
   shapeType=Capsule
   length=2.0185
   radius=1.1912
   shapeName=dyn_hair_01_front_r_capsule
   center=<2.044 -0.368 0>
   rotation=<rot 82.0913 180 180>
  }]
 dyn_hair_01_side_l [{shapeType=Capsule, length=1.4494, radius=1.5068, shapeName=dyn_hair_01_side_l_capsule, center=<1.704 0 0>, rotation=<rot 90 0 0>}]
 dyn_hair_01_side_r [{shapeType=Capsule, length=1.9789, radius=1.3701, shapeName=dyn_hair_01_side_r_capsule, center=<1.9235 0 0>, rotation=<rot 90 0 0>}]
 dyn_hair_02_front_l [{shapeType=Capsule, length=0.5061, radius=1.25, shapeName=dyn_hair_02_front_l_capsule, center=<1.0008 0 0>, rotation=<rot 90 0 0>}]
 dyn_hair_02_front_r [{
   shapeType=Capsule
   length=2.4856
   radius=1.1977
   shapeName=dyn_hair_02_front_r_capsule
   center=<1.9521 -0.6038 0.3004>
   rotation=<rot 74.4143 152.5417 179.9999>
  }]
 dyn_hair_02_side_l [{shapeType=Capsule, length=1.1771, radius=1.8627, shapeName=dyn_hair_02_side_l_capsule, center=<2.434 0 0>, rotation=<rot 90 0 0>}]
 dyn_hair_02_side_r [{shapeType=Capsule, length=2.1969, radius=1.4946, shapeName=dyn_hair_02_side_r_capsule, center=<1.8481 0 0>, rotation=<rot 90 0 0>}]
 dyn_hair_03_front_l [{shapeType=Capsule, length=1.1187, radius=1.25, shapeName=dyn_hair_03_front_l_capsule, center=<1.139 0 0>, rotation=<rot 90 0 0>}]
 dyn_hair_03_front_r [{shapeType=Capsule, length=1.3125, radius=1.0625, shapeName=dyn_hair_03_front_r_capsule, center=<0.8143 0 0>, rotation=<rot 90 0 0>}]
 dyn_hair_base_bk [{shapeType=Sphere, radius=0.5, shapeName=dyn_hair_base_bk_sphere}]
 dyn_hair_base_bk_side_l [{shapeType=Sphere, radius=0.5, shapeName=dyn_hair_base_bk_side_l_sphere}]
 dyn_hair_base_bk_side_r [{shapeType=Sphere, radius=0.5, shapeName=dyn_hair_base_bk_side_r_sphere}]
 dyn_hair_base_front_l [{shapeType=Sphere, radius=0.5, shapeName=dyn_hair_base_front_l_sphere}]
 dyn_hair_base_front_r [{shapeType=Sphere, radius=0.5, shapeName=dyn_hair_base_front_r_sphere}]
 dyn_hair_base_side_l [{shapeType=Sphere, radius=0.5, shapeName=dyn_hair_base_side_l_sphere}]
 dyn_hair_base_side_r [{shapeType=Sphere, radius=0.5, shapeName=dyn_hair_base_side_r_sphere}]
 dyn_hair_offset [{shapeType=Capsule, length=0.5625, radius=0.5, shapeName=dyn_hair_offset_capsule, rotation=<rot 90 0 0>}]
 dyn_hair_parent_bk [{shapeType=Capsule, length=2.7814, radius=1.3582, shapeName=dyn_hair_parent_bk_capsule, center=<1.6374 0 0>, rotation=<rot 90 0 0>}]
 dyn_hair_parent_bk_side_l [{
   shapeType=Capsule
   length=2.6429
   radius=1.6229
   shapeName=dyn_hair_parent_bk_side_l_capsule
   center=<2.1252 0 1.0233>
   rotation=<rot 64.367 180 -180>
  }]
 dyn_hair_parent_bk_side_r [{
   shapeType=Capsule
   length=2.1844
   radius=1.7859
   shapeName=dyn_hair_parent_bk_side_r_capsule
   center=<1.5826 0 0.6234>
   rotation=<rot 82.3669 180 180>
  }]
 dyn_hair_parent_front_l [{shapeType=Capsule, length=2.4617, radius=1.1461, shapeName=dyn_hair_parent_front_l_capsule, center=<1.8929 0 0>, rotation=<rot 90 0 0>}]
 dyn_hair_parent_front_r [{shapeType=Capsule, length=1.8104, radius=1.2812, shapeName=dyn_hair_parent_front_r_capsule, center=<1.9062 0 0>, rotation=<rot 90 0 0>}]
 dyn_hair_parent_side_l [{shapeType=Capsule, length=2.5073, radius=1.0908, shapeName=dyn_hair_parent_side_l_capsule, center=<1.7712 0 0>, rotation=<rot 90 0 0>}]
 dyn_hair_parent_side_r [{shapeType=Capsule, length=2.5439, radius=1.382, shapeName=dyn_hair_parent_side_r_capsule, center=<2.092 0 0>, rotation=<rot 90 0 0>}]
 dyn_strand_01 [{shapeType=Capsule, length=1.7343, radius=0.744, shapeName=dyn_strand_01_capsule, center=<1.0375 0 0>, rotation=<rot 90 0 0>}]
 dyn_strand_02 [{shapeType=Capsule, length=1.6737, radius=0.625, shapeName=dyn_strand_02_capsule, center=<1.0375 0 0>, rotation=<rot 90 0 0>}]
 dyn_strand_base [{shapeType=Sphere, radius=0.5, shapeName=dyn_strand_base_sphere}]
 dyn_strand_parent [{shapeType=Capsule, length=1.7432, radius=0.9249, shapeName=dyn_strand_parent_capsule, center=<1.0375 0 0>, rotation=<rot 90 0 0>}]
 head [{shapeType=Capsule, length=4.0625, radius=7.2812, shapeName=head_capsule, center=<3.2844 3.2394 0>, rotation=<rot 90 -172.875 165.402>}]
}
constraints [
 {bone1Name=dyn_strand_parent, bone2Name=dyn_strand_base, swing1LimitDegrees=10, swing2LimitDegrees=10, twistLimitDegrees=3}
 {bone1Name=dyn_strand_01, bone2Name=dyn_strand_parent, swing1LimitDegrees=10, swing2LimitDegrees=10, twistLimitDegrees=3}
 {bone1Name=dyn_strand_02, bone2Name=dyn_strand_01, swing1LimitDegrees=10, swing2LimitDegrees=10, twistLimitDegrees=3}
 {bone1Name=dyn_hair_parent_front_r, bone2Name=dyn_hair_base_front_r, swing1LimitDegrees=2, swing2LimitDegrees=1, twistLimitDegrees=3}
 {bone1Name=dyn_hair_01_front_r, bone2Name=dyn_hair_parent_front_r, swing1LimitDegrees=3, swing2LimitDegrees=1, twistLimitDegrees=3}
 {bone1Name=dyn_hair_02_front_r, bone2Name=dyn_hair_01_front_r, swing1LimitDegrees=3, swing2LimitDegrees=3, twistLimitDegrees=3}
 {bone1Name=dyn_hair_03_front_r, bone2Name=dyn_hair_02_front_r, swing1LimitDegrees=3, swing2LimitDegrees=3, twistLimitDegrees=3}
 {bone1Name=dyn_hair_parent_front_l, bone2Name=dyn_hair_base_front_l, swing1LimitDegrees=2, swing2LimitDegrees=2, twistLimitDegrees=3}
 {bone1Name=dyn_hair_01_front_l, bone2Name=dyn_hair_parent_front_l, swing1LimitDegrees=2, swing2LimitDegrees=2, twistLimitDegrees=3}
 {bone1Name=dyn_hair_02_front_l, bone2Name=dyn_hair_01_front_l, swing1LimitDegrees=3, swing2LimitDegrees=3, twistLimitDegrees=3}
 {bone1Name=dyn_hair_03_front_l, bone2Name=dyn_hair_02_front_l, swing1LimitDegrees=3, swing2LimitDegrees=3, twistLimitDegrees=3}
 {bone1Name=dyn_hair_parent_side_l, bone2Name=dyn_hair_base_side_l, swing1LimitDegrees=2, swing2LimitDegrees=2, twistLimitDegrees=2}
 {bone1Name=dyn_hair_01_side_l, bone2Name=dyn_hair_parent_side_l, swing1LimitDegrees=2, swing2LimitDegrees=2, twistLimitDegrees=2}
 {bone1Name=dyn_hair_02_side_l, bone2Name=dyn_hair_01_side_l, swing1LimitDegrees=2, swing2LimitDegrees=2, twistLimitDegrees=3}
 {bone1Name=dyn_hair_parent_side_r, bone2Name=dyn_hair_base_side_r, swing1LimitDegrees=5, swing2LimitDegrees=5, twistLimitDegrees=3}
 {bone1Name=dyn_hair_01_side_r, bone2Name=dyn_hair_parent_side_r, swing1LimitDegrees=5, swing2LimitDegrees=5, twistLimitDegrees=3}
 {bone1Name=dyn_hair_02_side_r, bone2Name=dyn_hair_01_side_r, swing1LimitDegrees=5, swing2LimitDegrees=5, twistLimitDegrees=3}
 {bone1Name=dyn_hair_parent_bk_side_l, bone2Name=dyn_hair_base_bk_side_l, swing1LimitDegrees=10, swing2LimitDegrees=10, twistLimitDegrees=3}
 {bone1Name=dyn_hair_01_bk_side_l, bone2Name=dyn_hair_parent_bk_side_l, swing1LimitDegrees=12, swing2LimitDegrees=10, twistLimitDegrees=3}
 {bone1Name=dyn_hair_parent_bk, bone2Name=dyn_hair_base_bk, swing1LimitDegrees=35, swing2LimitDegrees=25, twistLimitDegrees=3}
 {bone1Name=dyn_hair_01_bk_side_r, bone2Name=dyn_hair_parent_bk_side_r, swing1LimitDegrees=10, swing2LimitDegrees=10, twistLimitDegrees=3}
 {bone1Name=dyn_hair_parent_bk_side_r, bone2Name=dyn_hair_base_bk_side_r, swing1LimitDegrees=10, swing2LimitDegrees=5, twistLimitDegrees=3}
]
