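# Post-copy Anim Blueprint for the M_MED_Glare face accessory that copies the attached parent mesh's pose and adds secondary motion to its earrings.
# A RigidBody node simulates the accessory physics asset in spine_05 base-bone space with damped world motion.
# Constraints then tie both earring base bones to the head.
# Heavily damped, angle-limited Anim Dynamics nodes swing each earring parent bone relative to spine_05; the event graph holds no logic.
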AnimBlueprint /BRCosmetics/Characters/Player/Male/Medium/Bodies/M_MED_Glare/Meshes/Parts/M_MED_Glare_FaceAcc_AnimBP
target_skeleton=&/M_MED_Glare_FaceAcc_Skeleton
graphs [
 AnimGraph {nodes [
   n1 = RigidBody(ComponentPose=Local To Component(LocalPose=Copy Pose From Mesh() {bUseAttachedParent=true})) {
    bClampLinearTranslationLimitToRefPose=true
    bOverrideWorldGravity=true
    overridePhysicsAsset=&/M_MED_Glare_FaceAcc_Physics
    simulationSpace=BaseBoneSpace
    alphaBoolBlend {blendOption=Linear}
    baseBoneRef {boneName=spine_05}
    componentAppliedLinearAccClamp=<6000 6000 6000>
    componentLinearAccScale=<0.1 0.1 0.15>
    componentLinearVelScale=<0.1 0.1 0.15>
    overrideWorldGravity {z=-980}
    simSpaceSettings {maxAngularAcceleration=100, maxAngularVelocity=1, maxLinearAcceleration=100, maxLinearVelocity=50, velocityScaleZ=0, worldAlpha=0.5}
   }
   n2 = "Constraint - dyn_earing_base_r (P)"!(ComponentPose=@n1, Alpha=1, ConstraintWeights_0=1, ConstraintWeights_1:float=0) -> (Pose) {
    alphaBoolBlend {blendOption=Linear}
    boneToModify {boneName=dyn_earing_base_r}
    constraintSetup [{targetBone {boneName=head}}]
    constraintWeights [1,0]
   }
   n3 = "Constraint - dyn_earing_base_l (P)"!(ComponentPose=@n2, Alpha=1, ConstraintWeights_0=1, ConstraintWeights_1:float=0) -> (Pose) {
    alphaBoolBlend {blendOption=Linear}
    boneToModify {boneName=dyn_earing_base_l}
    constraintSetup [{targetBone {boneName=head}}]
    constraintWeights [1,0]
   }
   n4 = Anim Dynamics - Bone: dyn_earing_parent_l(ComponentPose=@n3, Alpha=1, ExternalForce="0,0,0") {
    angularDampingOverride=0.9999
    bOverrideAngularDamping=true
    bOverrideLinearDamping=true
    linearDampingOverride=0.9999
    alphaBoolBlend {blendOption=Linear}
    boundBone {boneName=dyn_earing_parent_l}
    componentAppliedLinearAccClamp=<6000 6000 6000>
    componentLinearAccScale=<0.09 0.09 0.09>
    componentLinearVelScale=<0.09 0.09 0.09>
    physicsBodyDefinitions [{
      boundBone {boneName=dyn_earing_parent_l}
      boxExtents=<10 5 5>
      constraintSetup {angularLimitsMax=<0 10 40>, angularLimitsMin=<0 -10 -40>, angularTarget=<0 0 0>}
      localJointOffset=<5 0 0>
     }]
    relativeSpaceBone {boneName=spine_05}
   }
   n5 = Anim Dynamics - Bone: dyn_earing_parent_r(ComponentPose=@n4, Alpha=1, ExternalForce="0,0,0") {
    angularDampingOverride=0.9999
    bOverrideAngularDamping=true
    bOverrideLinearDamping=true
    linearDampingOverride=0.9999
    alphaBoolBlend {blendOption=Linear}
    boundBone {boneName=dyn_earing_parent_r}
    componentAppliedLinearAccClamp=<6000 6000 6000>
    componentLinearAccScale=<0.09 0.09 0.09>
    componentLinearVelScale=<0.09 0.09 0.09>
    physicsBodyDefinitions [{
      boundBone {boneName=dyn_earing_parent_r}
      boxExtents=<10 5 5>
      constraintSetup {angularLimitsMax=<0 10 40>, angularLimitsMin=<0 -10 -40>, angularTarget=<0 0 0>}
      localJointOffset=<5 0 0>
     }]
    relativeSpaceBone {boneName=spine_05}
   }
   Output Pose(Result=Component To Local(ComponentPose=@n5)) {layerGroup=None}
  ]}
 EventGraph {nodes [Event BlueprintUpdateAnimation()]}
]
variables []
